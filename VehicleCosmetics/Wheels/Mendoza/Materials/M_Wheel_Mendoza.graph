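# Masked, lit material for the Mendoza vehicle wheel, built on MF_WheelParent_Simple with tintable emissive and central-element areas.
# Rim_mask green blends Emissive_color into the parent's emissive and red blends CentralElement_Color into its base color, with vertex-color red restoring parent values.
# Both tints are scaled by Emissive_int, so that parameter also affects base color.
# The result passes through MF_BaseVehicle_Damage, then MF_VehicleDeRez_Transition.
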Material /VehicleCosmetics/Wheels/Mendoza/Materials/M_Wheel_Mendoza
graph {
 properties {
  bAutomaticallySetUsageInEditor=false
  bUseMaterialAttributes=true
  bUsedWithMeshDeformer=true
  bUsedWithSkeletalMesh=true
  blendMode=BLEND_Masked
  pixelDepthOffsetMode=PDOM_Legacy
  previewMesh=/VehicleCosmetics/Wheels/Mendoza/Meshes/SK_Whee_Mendoza
  usedShadingModels=Default Lit
 }
 nodes [
  n1 = GetMaterialAttributes(None=MaterialFunctionCall() {materialFunction=&MF_WheelParent_Simple}.MaterialAttributes) {attributeGetTypes [#1, #2]}
  n2 = TextureSampleParameter2D() {parameterName=Rim_mask, samplerType=SAMPLERTYPE_LinearColor, texture=/VehicleCosmetics/Wheels/Mendoza/Textures/T_Mendoza_Rim_M3}
  n3 = OneMinus(None=@n2.G)
  n4 = Subtract(A=VectorParameter() {parameterName=Emissive_color, defaultValue=<rgba 1 0.7381 0 1>}.RGB, B=@n3)
  n5 = ScalarParameter() {defaultValue=1, parameterName=Emissive_int}
  n6 = NamedRerouteUsage()
  n7 = LinearInterpolate(A=LinearInterpolate(A=Multiply(A=@n4, B=@n5), B=@n6, Alpha=@n3), B=@n6, Alpha=VertexColor().R) {constB=0}
  n8 = OneMinus(None=@n2.R)
  n9 = Subtract(A=VectorParameter() {parameterName=CentralElement_Color, defaultValue=<rgba 1 0.7381 0 1>}.RGB, B=@n8)
  n10 = ScalarParameter() {defaultValue=1, parameterName=Emissive_int}
  n11 = NamedRerouteUsage()
  n12 = LinearInterpolate(A=LinearInterpolate(A=Multiply(A=@n9, B=@n10), B=@n11, Alpha=@n8), B=@n11, Alpha=VertexColor().R) {constB=0}
  MaterialFunctionCall(In=MaterialFunctionCall(Material=SetMaterialAttributes!(MaterialAttributes=@n1.MaterialAttributes, Emissive Color=@n7, Base Color=@n12) {attributeSetTypes [#1, #2]}) {materialFunction=&MF_BaseVehicle_Damage}.Result) {materialFunction=&MF_VehicleDeRez_Transition}
  NamedRerouteDeclaration(None=@n1.Emissive Color) {name=emiss, nodeColor=<rgba 0 1 0.7422 1>}
  NamedRerouteDeclaration(None=@n1.Base Color) {name=BC, nodeColor=<rgba 1 0.4219 0 1>}
 ]
}
